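# Looping sound cue for an Athena grenade launcher projectile in flight, routed to a medium explosives sound class through a Doppler node.
# It mixes an in-air loop with a shorter-range fuse group, each under its own HRTF-spatialized attenuation.
# The fuse group blends a fuse loop with a pre-explosion warning loop, both pitch- and volume-modulated by the PreExplo parameter.
# Concurrency limits simultaneous fuse sounds, with a stop-farthest resolution rule.
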SoundCue /Game/Sounds/Fort_Weapons/Launchers/Weapon_LAU_Grenade_Launcher_Athena_RocketInAir_Cue
root=SoundClass_2
nodes [
 {
  id=Attenuation_0
  children [WavePlayer_2]
  properties {
   attenuationSettings=&Fort_Weapon_Projectile_Medium_Mono_Athena_Att
   bOverrideAttenuation=true
   attenuationOverrides {
    distanceAlgorithm=Logarithmic
    falloffDistance=1536
    spatializationAlgorithm=SPATIALIZATION_HRTF
    stereoSpread=0
    attenuationShapeExtents {x=512}
   }
  }
 }
 {
  id=Attenuation_2
  children [Mixer_1]
  properties {
   attenuationSettings=&Fort_Weapon_Projectile_Medium_Mono_Athena_Att
   bOverrideAttenuation=true
   attenuationOverrides {
    bEnableReverbSend=false
    distanceAlgorithm=Logarithmic
    falloffDistance=250
    spatializationAlgorithm=SPATIALIZATION_HRTF
    stereoSpread=0
    attenuationShapeExtents {x=412}
   }
  }
 }
 {id=Doppler_2, children [Mixer_0], properties {dopplerIntensity=2}}
 {id=Mixer_0, children [Attenuation_0, Attenuation_2], properties {inputVolume [1,1]}}
 {id=Mixer_1, children [ModulatorContinuous_0, ModulatorContinuous_1], properties {inputVolume [0.75,1.5]}}
 {
  id=ModulatorContinuous_0
  children [Modulator_0]
  properties {
   pitchModulationParams {default=0, maxOutput=1.5, minOutput=1, parameterName=PreExplo}
   volumeModulationParams {default=0, maxOutput=1.2, minOutput=1, parameterName=PreExplo}
  }
 }
 {
  id=ModulatorContinuous_1
  children [WavePlayer_0]
  properties {
   pitchModulationParams {default=0, maxOutput=2, minOutput=1, parameterName=PreExplo}
   volumeModulationParams {default=0, parameterName=PreExplo}
  }
 }
 {id=Modulator_0, children [WavePlayer_1]}
 {id=SoundClass_2, children [Doppler_2], properties {soundClassOverride=/Game/Sounds/FortSoundClassesAndModes/Fort_Explosives_Medium_PSM}}
 {
  id=WavePlayer_0
  properties {bLooping=true, soundWaveAssetPtr=/Game/Sounds/Fort_GamePlay_Sounds/Explosives/GrenadeLauncher_Fuse_PreExploWarning_01}
 }
 {id=WavePlayer_1, properties {bLooping=true, soundWaveAssetPtr=/Game/Sounds/Fort_GamePlay_Sounds/Explosives/fuse_m80_loop_shortened}}
 {id=WavePlayer_2, properties {bLooping=true, soundWaveAssetPtr=&/Grenade_InAirLoop_Small_01}}
]
properties {
 maxDistance=2048
 volumeMultiplier=0.9
 attenuationOverrides {falloffDistance=5000, attenuationShapeExtents {x=1000}}
 concurrencyOverrides {
  bVolumeScaleCanRelease=true
  maxCount=2
  resolutionRule=StopFarthestThenPreventNew
  volumeScale=0.2
  volumeScaleAttackTime=0.2
  volumeScaleReleaseTime=2
 }
 concurrencySet [/Game/Sounds/Concurrency_Rules/Athena_Explosive_Fuse_Conc]
}
