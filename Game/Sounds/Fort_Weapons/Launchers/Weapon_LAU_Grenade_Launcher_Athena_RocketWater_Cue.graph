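# Sound cue for an Athena grenade launcher projectile in water, pairing a splash impact with a bubbling fuse.
# A mixer layers one of three equally weighted random water impact one-shots with a looping underwater fuse bubble wave.
# The mix passes through an overridden attenuation node with short-range logarithmic falloff and HRTF spatialization, then a medium-explosives sound class node.
# The looping layer keeps the cue playing until stopped, and a shared explosive-fuse concurrency set governs simultaneous instances.
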SoundCue /Game/Sounds/Fort_Weapons/Launchers/Weapon_LAU_Grenade_Launcher_Athena_RocketWater_Cue
root=SoundClass_0
nodes [
 {
  id=Attenuation_3
  children [Mixer_2]
  properties {
   attenuationSettings=&Fort_Weapon_Projectile_Medium_Mono_Athena_Att
   bOverrideAttenuation=true
   attenuationOverrides {
    bEnableReverbSend=false
    distanceAlgorithm=Logarithmic
    falloffDistance=250
    spatializationAlgorithm=SPATIALIZATION_HRTF
    stereoSpread=0
    attenuationShapeExtents {x=412}
   }
  }
 }
 {id=Mixer_2, children [Random_0,WavePlayer_4], properties {inputVolume [1,1]}}
 {id=Random_0, children [WavePlayer_6,WavePlayer_7,WavePlayer_8], properties {weights [1,1,1]}}
 {id=SoundClass_0, children [Attenuation_3], properties {soundClassOverride=/Game/Sounds/FortSoundClassesAndModes/Fort_Explosives_Medium_PSM}}
 {
  id=WavePlayer_4
  properties {bLooping=true, soundWaveAssetPtr=/Game/Sounds/Fort_GamePlay_Sounds/Explosives/Grenade_Proj_Fuse_Water_Bubbles_Loop}
 }
 {id=WavePlayer_6, properties {soundWaveAssetPtr=/Game/Sounds/Fort_GamePlay_Sounds/Explosives/Grenade_Impact_Water_01}}
 {id=WavePlayer_7, properties {soundWaveAssetPtr=/Game/Sounds/Fort_GamePlay_Sounds/Explosives/Grenade_Impact_Water_02}}
 {id=WavePlayer_8, properties {soundWaveAssetPtr=/Game/Sounds/Fort_GamePlay_Sounds/Explosives/Grenade_Impact_Water_03}}
]
properties {
 maxDistance=662
 volumeMultiplier=0.9
 attenuationOverrides {falloffDistance=5000, attenuationShapeExtents {x=1000}}
 concurrencyOverrides {
  bVolumeScaleCanRelease=true
  maxCount=2
  resolutionRule=StopFarthestThenPreventNew
  volumeScale=0.2
  volumeScaleAttackTime=0.2
  volumeScaleReleaseTime=2
 }
 concurrencySet [/Game/Sounds/Concurrency_Rules/Athena_Explosive_Fuse_Conc]
}
