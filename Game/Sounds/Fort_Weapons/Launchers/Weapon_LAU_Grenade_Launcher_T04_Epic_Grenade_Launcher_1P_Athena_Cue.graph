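# Sound cue for the first-person fire sound of the T04 Epic grenade launcher in Athena.
# A single grenade fire wave plays through a Modulator at default settings, with the cue volume slightly reduced.
# It uses the first-person projectile attenuation asset and shared global and local gunfire concurrency sets.
# Despite the 1P name, it is assigned to the third-person ranged-fire sound class.
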SoundCue /Game/Sounds/Fort_Weapons/Launchers/Weapon_LAU_Grenade_Launcher_T04_Epic_Grenade_Launcher_1P_Athena_Cue
root=Modulator_15
nodes [
 {id=Modulator_15, children [WavePlayer_11]}
 {id=WavePlayer_11, properties {soundWaveAssetPtr=&/weapon_grenade_fire_04}}
]
properties {
 attenuationSettings=&Fort_Weapon_Projectile_FirstPerson_Athena_Att
 maxDistance=2097152
 soundClassObject=&Fort_WeaponRanged_Fire_3P_PSM
 volumeMultiplier=0.8
 attenuationOverrides {falloffDistance=5000, attenuationShapeExtents {x=1000}}
 concurrencyOverrides {bLimitToOwner=true, maxCount=3, resolutionRule=StopOldest}
 concurrencySet [&Weapon_Gunfire_Global_Conc, &Weapon_Gunfire_Local_Conc]
}
