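# Sound cue for the end of an RPG launcher weapon reload, classed as weapon foley.
# A single launcher reload-end wave player feeds a modulator at the root, with no random selection.
# It references the shared weapon reload attenuation settings and overrides concurrency to allow one instance, stopping the farthest before preventing new ones.
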SoundCue /Game/Sounds/Fort_Weapons/Launchers/Weapon_LAU_RPG_Reload_End_Cue
root=Modulator_1
nodes [
 {id=Modulator_1, children [WavePlayer_4]}
 {id=WavePlayer_4, properties {soundWaveAssetPtr=&/weapons_launchers_reload_end_01}}
]
properties {
 attenuationSettings=&Fort_Weapon_Reload_Att
 bOverrideConcurrency=true
 maxDistance=2250
 soundClassObject=&Fort_WeaponFoley
 volumeMultiplier=1.002
 attenuationOverrides {falloffDistance=5000, attenuationShapeExtents {x=1000}}
 concurrencyOverrides {maxCount=1, resolutionRule=StopFarthestThenPreventNew}
}
