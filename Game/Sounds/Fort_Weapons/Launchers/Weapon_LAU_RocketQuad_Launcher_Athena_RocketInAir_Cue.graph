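# Sound cue for the in-air sound of the Athena quad rocket launcher's rockets, with per-owner concurrency limiting.
# Each quality branch mixes an enveloped projectile one-shot with a quieter, pitch-modulated distant loop that fades in, each using its own sound class and attenuation.
# The first two quality levels randomize among three quad-rocket waves, while the last two swap in a single optimized projectile wave and an optimized loop.
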SoundCue /Game/Sounds/Fort_Weapons/Launchers/Weapon_LAU_RocketQuad_Launcher_Athena_RocketInAir_Cue
root=QualityLevel_0
nodes [
 {id=Attenuation_1, children [Modulator_0], properties {attenuationSettings=/Game/Sounds/Attenuation/Fort_Weapon_Rocket_Distant_Athena_Att}}
 {
  id=Attenuation_10
  children [Enveloper_5]
  properties {
   attenuationSettings=&Fort_Weapon_Projectile_Medium_Mono_Athena_Att
   bOverrideAttenuation=true
   attenuationOverrides {falloffDistance=10000, stereoSpread=512, attenuationShapeExtents {x=2000}}
  }
 }
 {
  id=Attenuation_11
  children [Modulator_7]
  properties {
   attenuationSettings=&Fort_Weapon_Rocket_Distant_Athena_Att
   attenuationOverrides {falloffDistance=2600, attenuationShapeExtents {x=1000}}
  }
 }
 {
  id=Attenuation_2
  children [Enveloper_1]
  properties {
   attenuationSettings=&Fort_Weapon_Projectile_Medium_Mono_Athena_Att
   bOverrideAttenuation=true
   attenuationOverrides {falloffDistance=10000, stereoSpread=512, attenuationShapeExtents {x=2000}}
  }
 }
 {
  id=Enveloper_0
  children [WavePlayer_1]
  properties {
   pitchCurve {editorCurveData {keys [{time=0}]}}
   volumeCurve {editorCurveData {keys [{interpMode=RCIM_Cubic, tangentMode=RCTM_User, time=0, value=0}, {interpMode=RCIM_Cubic, tangentMode=RCTM_User, time=1, value=1}]}}
  }
 }
 {
  id=Enveloper_1
  children [WavePlayer_4]
  properties {
   pitchCurve {editorCurveData {keys [{time=0}]}}
   volumeCurve {editorCurveData {keys [
      {interpMode=RCIM_Cubic, tangentMode=RCTM_User, time=0, value=0}
      {interpMode=RCIM_Cubic, tangentMode=RCTM_User, time=0.25, value=1}
      <at 5 0>
     ]}}
  }
 }
 {
  id=Enveloper_4
  children [WavePlayer_0]
  properties {
   pitchCurve {editorCurveData {keys [{time=0}]}}
   volumeCurve {editorCurveData {keys [{interpMode=RCIM_Cubic, tangentMode=RCTM_User, time=0, value=0}, {interpMode=RCIM_Cubic, tangentMode=RCTM_User, time=1, value=1}]}}
  }
 }
 {
  id=Enveloper_5
  children [Random_0]
  properties {
   pitchCurve {editorCurveData {keys [{time=0}]}}
   volumeCurve {editorCurveData {keys [
      {interpMode=RCIM_Cubic, tangentMode=RCTM_User, time=0, value=0}
      {interpMode=RCIM_Cubic, tangentMode=RCTM_User, time=0.25, value=1}
      <at 5 0>
     ]}}
  }
 }
 {id=Mixer_0, children [SoundClass_1,SoundClass_0], properties {inputVolume [0.5,0.2]}}
 {id=Mixer_3, children [SoundClass_2,SoundClass_6], properties {inputVolume [0.5,0.2]}}
 {id=Modulator_0, children [Enveloper_0], properties {pitchMax=1.1}}
 {id=Modulator_7, children [Enveloper_4], properties {pitchMax=1.1}}
 {id=QualityLevel_0, children [Mixer_3,Mixer_3,Mixer_0,Mixer_0]}
 {id=Random_0, children [WavePlayer_2,WavePlayer_5,WavePlayer_6], properties {weights [1,1,1]}}
 {
  id=SoundClass_0
  children [Attenuation_1]
  properties {soundClassOverride=&Fort_WeaponRanged_Fire_Distant}
 }
 {
  id=SoundClass_1
  children [Attenuation_2]
  properties {soundClassOverride=&Fort_WeaponRanged_Fire_3P_PSM}
 }
 {
  id=SoundClass_2
  children [Attenuation_10]
  properties {soundClassOverride=&Fort_WeaponRanged_Fire_3P_PSM}
 }
 {
  id=SoundClass_6
  children [Attenuation_11]
  properties {soundClassOverride=&Fort_WeaponRanged_Fire_Distant}
 }
 {id=WavePlayer_0, properties {bLooping=true, soundWaveAssetPtr=&/RPG_InAir_NewLoop_LP2}}
 {id=WavePlayer_1, properties {bLooping=true, soundWaveAssetPtr=&/RPG_InAir_NewLoop_LP2_opt}}
 {id=WavePlayer_2, properties {soundWaveAssetPtr=/Game/Sounds/Fort_Weapons/Launchers/Quad/quad_rocket_inair_fire_01}}
 {id=WavePlayer_4, properties {soundWaveAssetPtr=&/Athena_New_Rocket_Projectile_opt}}
 {id=WavePlayer_5, properties {soundWaveAssetPtr=/Game/Sounds/Fort_Weapons/Launchers/Quad/quad_rocket_inair_fire_02}}
 {id=WavePlayer_6, properties {soundWaveAssetPtr=/Game/Sounds/Fort_Weapons/Launchers/Quad/quad_rocket_inair_fire_03}}
]
properties {
 bOverrideConcurrency=true
 maxDistance=15000
 soundClassObject=&Fort_WeaponRanged_Fire_3P_PSM
 volumeMultiplier=1
 attenuationOverrides {falloffDistance=5000, attenuationShapeExtents {x=1000}}
 concurrencyOverrides {bLimitToOwner=true, maxCount=3, resolutionRule=StopQuietest}
}
