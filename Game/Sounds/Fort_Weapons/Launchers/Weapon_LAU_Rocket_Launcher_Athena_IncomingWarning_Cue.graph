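# Sound cue for the incoming rocket warning of the Athena rocket launcher, playing a projectile sound.
# A quality-level switch sends the first two levels to the standard wave and the last two to the "opt" wave.
# Each wave runs through a quick fade-in and longer fade-out envelope, sound class override, modulator and mixer with an empty second input.
# Overrides set attenuation and a per-owner voice limit that stops the quietest.
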SoundCue /Game/Sounds/Fort_Weapons/Launchers/Weapon_LAU_Rocket_Launcher_Athena_IncomingWarning_Cue
root=QualityLevel_0
nodes [
 {
  id=Enveloper_1
  children [WavePlayer_4]
  properties {
   pitchCurve {editorCurveData {keys [{time=0}]}}
   volumeCurve {editorCurveData {keys [
      {interpMode=RCIM_Cubic, tangentMode=RCTM_User, time=0, value=0.01}
      {interpMode=RCIM_Cubic, tangentMode=RCTM_User, time=0.25, value=1}
      <at 5 0>
     ]}}
  }
 }
 {
  id=Enveloper_5
  children [WavePlayer_3]
  properties {
   pitchCurve {editorCurveData {keys [{time=0}]}}
   volumeCurve {editorCurveData {keys [
      {interpMode=RCIM_Cubic, tangentMode=RCTM_User, time=0, value=0.01}
      {interpMode=RCIM_Cubic, tangentMode=RCTM_User, time=0.25, value=1}
      <at 5 0>
     ]}}
  }
 }
 {id=Mixer_0, children [Modulator_2,null], properties {inputVolume [0.5,0.2]}}
 {id=Mixer_3, children [Modulator_1,null], properties {inputVolume [0.5,0.2]}}
 {id=Modulator_1, children [SoundClass_2]}
 {id=Modulator_2, children [SoundClass_1]}
 {id=QualityLevel_0, children [Mixer_3,Mixer_3,Mixer_0,Mixer_0]}
 {id=SoundClass_1, children [Enveloper_1], properties {soundClassOverride=&Fort_WeaponRanged_Fire_3P_PSM}}
 {id=SoundClass_2, children [Enveloper_5], properties {soundClassOverride=&Fort_WeaponRanged_Fire_3P_PSM}}
 {id=WavePlayer_3, properties {soundWaveAssetPtr=&/Athena_New_Rocket_Projectile_01}}
 {id=WavePlayer_4, properties {soundWaveAssetPtr=&/Athena_New_Rocket_Projectile_opt}}
]
properties {
 bOverrideAttenuation=true
 bOverrideConcurrency=true
 maxDistance=7000
 soundClassObject=&Fort_WeaponRanged_Fire_3P_PSM
 volumeMultiplier=0.6
 attenuationOverrides {falloffDistance=7000, stereoSpread=1500, attenuationShapeExtents {x=0}}
 concurrencyOverrides {bLimitToOwner=true, maxCount=3, resolutionRule=StopQuietest}
}
