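# Sound cue for the incoming-rocket warning of the low-tier rocket launcher in Athena.
# Randomly picks one of two warning waves and shapes it with a volume envelope that swells in quickly, then fades to silence by five seconds.
# The result passes through a third-person ranged-fire sound class override, a modulator and a mixer.
# Cue-level overrides set spatial falloff and cap simultaneous instances per owner, stopping the quietest.
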SoundCue /Game/Sounds/Fort_Weapons/Launchers/Weapon_LAU_Rocket_Launcher_Athena_IncomingWarning_LowTier_Cue
root=Mixer_3
nodes [
 {
  id=Enveloper_5
  children [Random_1]
  properties {
   pitchCurve {editorCurveData {keys [{time=0}]}}
   volumeCurve {editorCurveData {keys [
      {interpMode=RCIM_Cubic, tangentMode=RCTM_User, time=0, value=0.01}
      {interpMode=RCIM_Cubic, tangentMode=RCTM_User, time=0.25, value=1}
      <at 5 0>
     ]}}
  }
 }
 {id=Mixer_3, children [Modulator_1], properties {inputVolume [0.75]}}
 {id=Modulator_1, children [SoundClass_2]}
 {id=Random_1, children [WavePlayer_2,WavePlayer_5], properties {weights [1,1]}}
 {id=SoundClass_2, children [Enveloper_5], properties {soundClassOverride=&Fort_WeaponRanged_Fire_3P_PSM}}
 {id=WavePlayer_2, properties {soundWaveAssetPtr=&/Wep_RPG_LowTier_IncomingWarning_01}}
 {id=WavePlayer_5, properties {soundWaveAssetPtr=&/Wep_RPG_LowTier_IncomingWarning_02}}
]
properties {
 bOverrideAttenuation=true
 bOverrideConcurrency=true
 maxDistance=7000
 soundClassObject=&Fort_WeaponRanged_Fire_3P_PSM
 volumeMultiplier=0.6
 attenuationOverrides {falloffDistance=7000, stereoSpread=1500, attenuationShapeExtents {x=0}}
 concurrencyOverrides {bLimitToOwner=true, maxCount=3, resolutionRule=StopQuietest}
}
