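# Sound cue for the in-flight sound of the Athena rocket launcher's projectile, spatialized with the launcher projectile in-air attenuation plus overrides.
# A mixer layers a looping wave, faded in by an enveloper, pitch-modulated and run through smoothed Doppler, with a louder, randomly chosen start sound from three variants.
# Concurrency is overridden to limit instances per owner, stopping the quietest when over the cap.
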SoundCue /Game/Sounds/Fort_Weapons/Launchers/Weapon_LAU_Rocket_Launcher_Athena_RocketInAir_Cue
root=Mixer_1
nodes [
 {id=Doppler_1, children [Modulator_7], properties {bUseSmoothing=true, dopplerIntensity=3}}
 {
  id=Enveloper_4
  children [WavePlayer_0]
  properties {
   pitchCurve {editorCurveData {keys [{time=0}]}}
   volumeCurve {editorCurveData {keys [{interpMode=RCIM_Cubic, tangentMode=RCTM_User, time=0, value=0}, {interpMode=RCIM_Cubic, tangentMode=RCTM_User, time=1, value=1}]}}
  }
 }
 {id=Mixer_1, children [Doppler_1,Modulator_1], properties {inputVolume [1,1.5]}}
 {id=Modulator_1, children [Random_0]}
 {id=Modulator_7, children [Enveloper_4], properties {pitchMax=1.1}}
 {id=Random_0, children [WavePlayer_2,WavePlayer_3,WavePlayer_4], properties {bShouldExcludeFromBranchCulling=true, weights [1,1,1]}}
 {id=WavePlayer_0, properties {bLooping=true, soundWaveAssetPtr=&/RPG_InAir_NewLoop_LP2}}
 {id=WavePlayer_2, properties {soundWaveAssetPtr=&/RPG_InAir_Prj_Loop_Start_01}}
 {id=WavePlayer_3, properties {soundWaveAssetPtr=&/RPG_InAir_Prj_Loop_Start_02}}
 {id=WavePlayer_4, properties {soundWaveAssetPtr=&/RPG_InAir_Prj_Loop_Start_03}}
]
properties {
 attenuationSettings=/Game/Sounds/Attenuation/Fort_Weapon_Launcher_Proj_InAir_Default_Att
 bOverrideConcurrency=true
 maxDistance=12750
 soundClassObject=&Fort_WeaponRanged_Fire_3P_PSM
 volumeMultiplier=0.5
 attenuationOverrides {falloffDistance=5000, attenuationShapeExtents {x=1000}}
 concurrencyOverrides {bLimitToOwner=true, maxCount=3, resolutionRule=StopQuietest}
}
