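# Sound cue for a rocket passby on the Athena rocket launcher, routed to the ranged weapon sound class.
# A Mixer layers two reused meteor "almost landed" waves at once, feeding a root Modulator at default settings.
# The cue plays at reduced volume, and its attenuation override turns off distance volume falloff.
# A concurrency override caps simultaneous instances per owner, stopping the quietest when exceeded.
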SoundCue /Game/Sounds/Fort_Weapons/Launchers/Weapon_LAU_Rocket_Launcher_Athena_RocketPassby_Cue
root=Modulator_0
nodes [
 {id=Mixer_1, children [WavePlayer_2,WavePlayer_5], properties {inputVolume [1,1]}}
 {id=Modulator_0, children [Mixer_1]}
 {id=WavePlayer_2, properties {soundWaveAssetPtr=/Game/Sounds/Fort_GamePlay_Sounds/Meteor/Meteor_Almost_Landed_01}}
 {id=WavePlayer_5, properties {soundWaveAssetPtr=/Game/Sounds/Fort_GamePlay_Sounds/Meteor/Meteor_Almost_Landed_02}}
]
properties {
 bOverrideAttenuation=true
 bOverrideConcurrency=true
 maxDistance=2097152
 soundClassObject=&Fort_WeaponRanged
 volumeMultiplier=0.25
 attenuationOverrides {bAttenuate=false, falloffDistance=5000, stereoSpread=1500, attenuationShapeExtents {x=1000}}
 concurrencyOverrides {bLimitToOwner=true, maxCount=3, resolutionRule=StopQuietest}
}
