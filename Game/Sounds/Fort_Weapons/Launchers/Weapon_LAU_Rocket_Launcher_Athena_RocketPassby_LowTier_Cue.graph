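# Sound cue for the close pass-by sound of a low-tier rocket launcher rocket in Athena.
# It picks one of two equally weighted wave variations at random and feeds it through a Modulator at default settings.
# It plays at reduced volume in the ranged weapon sound class, with overrides that disable distance attenuation and cap concurrency per owner, stopping the quietest instance.
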SoundCue /Game/Sounds/Fort_Weapons/Launchers/Weapon_LAU_Rocket_Launcher_Athena_RocketPassby_LowTier_Cue
root=Modulator_0
nodes [
 {id=Modulator_0, children [Random_0]}
 {id=Random_0, children [WavePlayer_0,WavePlayer_1], properties {weights [1,1]}}
 {id=WavePlayer_0, properties {soundWaveAssetPtr=&/Wep_RPG_LowTier_PassBy_Close_01}}
 {id=WavePlayer_1, properties {soundWaveAssetPtr=&/Wep_RPG_LowTier_PassBy_Close_02}}
]
properties {
 bOverrideAttenuation=true
 bOverrideConcurrency=true
 maxDistance=2097152
 soundClassObject=&Fort_WeaponRanged
 volumeMultiplier=0.4
 attenuationOverrides {bAttenuate=false, falloffDistance=5000, stereoSpread=1500, attenuationShapeExtents {x=1000}}
 concurrencyOverrides {bLimitToOwner=true, maxCount=3, resolutionRule=StopQuietest}
}
