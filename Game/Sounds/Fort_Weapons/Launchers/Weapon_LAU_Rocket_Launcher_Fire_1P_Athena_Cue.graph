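# First-person rocket launcher firing sound cue for Athena.
# A Random node picks between two equally weighted launch waves and is excluded from branch culling, feeding a Modulator that holds pitch fixed.
# The cue boosts volume and uses the first-person ranged fire sound class, projectile attenuation and shared first-person gunfire concurrency.
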SoundCue /Game/Sounds/Fort_Weapons/Launchers/Weapon_LAU_Rocket_Launcher_Fire_1P_Athena_Cue
root=Modulator_6
nodes [
 {id=Modulator_6, children [Random_3], properties {pitchMax=1, pitchMin=1}}
 {id=Random_3, children [WavePlayer_18, WavePlayer_19], properties {bShouldExcludeFromBranchCulling=true, weights [1,1]}}
 {id=WavePlayer_18, properties {soundWaveAssetPtr=&/rocket_launcher_start_02}}
 {id=WavePlayer_19, properties {soundWaveAssetPtr=&/rocket_launcher_start_01}}
]
properties {
 attenuationSettings=&Fort_Weapon_Projectile_FirstPerson_Athena_Att
 maxDistance=2097152
 soundClassObject=&Fort_WeaponRanged_Fire_1P_PSM
 volumeMultiplier=1.5
 attenuationOverrides {falloffDistance=5000, attenuationShapeExtents {x=1000}}
 concurrencyOverrides {bLimitToOwner=true, maxCount=3, resolutionRule=PreventNew}
 concurrencySet [&Weapon_Gunfire_FirstPerson_Conc]
}
