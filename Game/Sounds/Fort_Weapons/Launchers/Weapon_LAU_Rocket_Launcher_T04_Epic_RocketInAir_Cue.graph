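# Sound cue for the Epic T04 rocket launcher, named for the rocket-in-air sound.
# It mixes a randomly picked tail wave, under the third-person fire class with low-pass attenuation, with a distant-fire class branch.
# That branch crossfades by distance from a medium shot to a pitch-modulated distant shot, each with its own attenuation.
# The cue overrides concurrency to prevent new instances once its limit is reached.
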SoundCue /Game/Sounds/Fort_Weapons/Launchers/Weapon_LAU_Rocket_Launcher_T04_Epic_RocketInAir_Cue
root=Mixer_3
nodes [
 {id=Attenuation_10, children [WavePlayer_42], properties {attenuationSettings=/Game/Sounds/Attenuation/Fort_Explosion_RPG_Att}}
 {id=Attenuation_11, children [Modulator_7], properties {attenuationSettings=/Game/Sounds/Attenuation/Fort_Weapon_Projectile_Distant_StW_Att}}
 {
  id=Attenuation_9
  children [Modulator_6]
  properties {
   attenuationSettings=&Fort_Explosion_RPG_Att
   attenuationOverrides {bAttenuateWithLPF=true, falloffDistance=3000, lPFRadiusMax=3000, lPFRadiusMin=1500, attenuationShapeExtents {x=2000}}
  }
 }
 {
  id=DistanceCrossFade_3
  children [Attenuation_10, Attenuation_11]
  properties {crossFadeInput [
    {fadeInDistanceEnd=100, fadeInDistanceStart=0, fadeOutDistanceEnd=3000, fadeOutDistanceStart=2000, volume=1}
    {fadeInDistanceEnd=3000, fadeInDistanceStart=1000, fadeOutDistanceEnd=12000, fadeOutDistanceStart=10000, volume=1}
   ]}
 }
 {id=Mixer_3, children [SoundClass_7,SoundClass_6], properties {inputVolume [1,1]}}
 {id=Modulator_6, children [Random_4], properties {pitchMax=1, pitchMin=1}}
 {id=Modulator_7, children [WavePlayer_45], properties {pitchMax=1.1}}
 {id=Random_4, children [WavePlayer_20, WavePlayer_21], properties {weights [1,1]}}
 {
  id=SoundClass_6
  children [DistanceCrossFade_3]
  properties {soundClassOverride=&Fort_WeaponRanged_Fire_Distant}
 }
 {
  id=SoundClass_7
  children [Attenuation_9]
  properties {soundClassOverride=&Fort_WeaponRanged_Fire_3P_PSM}
 }
 {id=WavePlayer_20, properties {soundWaveAssetPtr=&/rocket_launcher_tail_02}}
 {id=WavePlayer_21, properties {soundWaveAssetPtr=&/rocket_launcher_tail_01}}
 {id=WavePlayer_42, properties {soundWaveAssetPtr=&/Rocket_Launcher_Medium_ShotC01}}
 {id=WavePlayer_45, properties {soundWaveAssetPtr=&/Rocket_Launcher_Distant_ShotE01}}
]
properties {
 bOverrideConcurrency=true
 maxDistance=20000
 volumeMultiplier=1.5
 attenuationOverrides {falloffDistance=5000, attenuationShapeExtents {x=1000}}
 concurrencyOverrides {resolutionRule=PreventNew}
}
